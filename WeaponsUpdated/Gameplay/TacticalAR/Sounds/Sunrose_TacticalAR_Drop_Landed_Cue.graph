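# Sound cue for the Sunrose Tactical AR weapon's drop-landed impact.
# A random node picks between two equally weighted impact waves and feeds a modulator left at default settings.
# It uses the shared weapon drop attenuation and weapon foley sound class, with reduced volume, a concurrency override and the debug flag left on.
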SoundCue /WeaponsUpdated/Gameplay/TacticalAR/Sounds/Sunrose_TacticalAR_Drop_Landed_Cue
root=Modulator_0
nodes [
 {id=Modulator_0, children [Random_0]}
 {id=Random_0, children [WavePlayer_2,WavePlayer_3], properties {weights [1,1]}}
 {id=WavePlayer_2, properties {soundWaveAssetPtr=&/Sunrose_TacticalAR_Drop_Impt_02}}
 {id=WavePlayer_3, properties {soundWaveAssetPtr=&/Sunrose_TacticalAR_Drop_Impt_01}}
]
properties {
 attenuationSettings=&Fort_Weapon_Drop_Att
 bDebug=true
 bOverrideConcurrency=true
 maxDistance=3000
 soundClassObject=&Fort_WeaponFoley
 volumeMultiplier=0.451
}
