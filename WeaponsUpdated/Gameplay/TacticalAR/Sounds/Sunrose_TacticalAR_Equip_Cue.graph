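# Sound cue for the equip sound of the Sunrose tactical assault rifle, reusing an assault rifle reload-finish wave.
# The wave passes through an enveloper that ramps volume up around a tenth of a second in, then a fixed short delay and a root modulator.
# It plays as weapon foley with reload attenuation and overridden concurrency, at reduced volume.
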SoundCue /WeaponsUpdated/Gameplay/TacticalAR/Sounds/Sunrose_TacticalAR_Equip_Cue
root=Modulator_0
nodes [
 {id=Delay_0, children [Enveloper_0], properties {delayMax=0.15, delayMin=0.15}}
 {
  id=Enveloper_0
  children [WavePlayer_0]
  properties {pitchCurve {editorCurveData {keys [{time=0}]}}, volumeCurve {editorCurveData {keys [<at 0.09 0.01>,<at 0.1 1>]}}}
 }
 {id=Modulator_0, children [Delay_0]}
 {id=WavePlayer_0, properties {soundWaveAssetPtr=/Game/Sounds/Fort_Weapons/AutoRifle/Assault_PistolCaliber/JW_AR_ReloadFinish_01}}
]
properties {
 attenuationSettings=&Fort_Weapon_Reload_Att
 bDebug=true
 bOverrideConcurrency=true
 maxDistance=2250
 soundClassObject=&Fort_WeaponFoley
 volumeMultiplier=0.501
}
