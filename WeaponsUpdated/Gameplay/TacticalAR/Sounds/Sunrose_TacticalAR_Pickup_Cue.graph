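# Sound cue for the Sunrose TacticalAR weapon pickup sound, routed to the Fort weapon foley sound class.
# A Random node picks between two equally weighted pickup waves and feeds a Modulator left at default settings.
# It uses the shared weapon pickup attenuation, lowers the volume to under half, overrides concurrency, and has its debug flag on.
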SoundCue /WeaponsUpdated/Gameplay/TacticalAR/Sounds/Sunrose_TacticalAR_Pickup_Cue
root=Modulator_0
nodes [
 {id=Modulator_0, children [Random_0]}
 {id=Random_0, children [WavePlayer_0,WavePlayer_1], properties {weights [1,1]}}
 {id=WavePlayer_0, properties {soundWaveAssetPtr=&/Sunrose_TacticalAR_Pickup_01}}
 {id=WavePlayer_1, properties {soundWaveAssetPtr=&/Sunrose_TacticalAR_Pickup_02}}
]
properties {
 attenuationSettings=&Fort_Weapon_Pickup_Att
 bDebug=true
 bOverrideConcurrency=true
 maxDistance=3500
 soundClassObject=&Fort_WeaponFoley
 volumeMultiplier=0.451
}
